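; Default elimination utility component for NPCs, with only its client and server initialization defined here.
; InitComponentClient and InitComponentServer each print a debug message controlled by bDebugComponent.
; Both check the owner against BP_PlayerPawn_NonParticipant by soft class, then cast and cache it as AIPawn.
; No elimination behavior appears in these functions.
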(fn InitComponentClient
  (bind _returnvalue (Components|GetOwner))
  (Development|PrintString "@@@ InitComponentClient [BP_EliminationUtilityComp_Default]: CALLED!" false (Variables|Debug|GetbDebugComponent))
  (Utilities|IsA(soft) _returnvalue "/Game/Athena/AI/NPCs/Base/BP_PlayerPawn_NonParticipant.BP_PlayerPawn_NonParticipant_C"
    (:True
      (bind _asbp_player_pawn_non_participant (Utilities|Casting|CastToBP_PlayerPawn_NonParticipant _returnvalue))
      (Variables|Internal|SetAIPawn _asbp_player_pawn_non_participant))))

(fn InitComponentServer
  (bind _returnvalue (Components|GetOwner))
  (Development|PrintString "@@@ InitComponentServer [BP_EliminationUtilityComp_Default]: CALLED!" false (Variables|Debug|GetbDebugComponent))
  (Utilities|IsA(soft) _returnvalue "/Game/Athena/AI/NPCs/Base/BP_PlayerPawn_NonParticipant.BP_PlayerPawn_NonParticipant_C"
    (:True
      (bind _asbp_player_pawn_non_participant (Utilities|Casting|CastToBP_PlayerPawn_NonParticipant _returnvalue))
      (Variables|Internal|SetAIPawn _asbp_player_pawn_non_participant))))
