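; Actor component that animates the vertical offset of an NPC boss's status widget, acting only on non-authority clients.
; UpdateVerticalOffset (when not overridden or animating) and ResetVerticalOffset (when not animating) start ease curve animations, enabling tick and the pawn's offset override.
; Tick adds the curve-scaled offset to the NPC widget offset as render translation.
; Completion disables tick and releases the override unless hold is set.
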(event EventBeginPlay
  (Utilities|SwitchHasAuthority
    (:Remote
      (bind _output_get (Variables|InternalData|SetNPCActor (Components|GetOwner self)))
      (bind _asbp_player_pawn_non_participant_boss (Utilities|Casting|CastToBP_PlayerPawn_NonParticipant_Boss _output_get))
      (bind _output_get_2 (Variables|InternalData|SetNPCBossPawn _asbp_player_pawn_non_participant_boss))
      (bind _npcstatuswidget (Variables|/Game/Athena/AI/NPCs/Base/BP_PlayerPawn_NonParticipant_Boss.BP_PlayerPawn_NonParticipant_Boss_C|GetNPCStatusWidget _output_get_2)
        (:then
          (Variables|InternalData|SetNPCStatusWidget_UserObj (UserInterface|GetUserWidgetObject _npcstatuswidget))
          (Variables|InternalData|SetbIsAnimating false)
          (Variables|WidgetOffset|SetOffsetY 0.0)
          (Variables|WidgetOffset|SetWidgetAnimTime 0.0)
          (Variables|WidgetOffset|SetWidgetAnimCurve 0)
          (Variables|WidgetOffset|SetbHoldWidgetOffset false))
        (:else)))))

(event Custom|UpdateVerticalOffset
  (params (Offset%20Y :type Float%20%28double-precision%29) (Animation%20Time :type Float%20%28double-precision%29) (Hold%20Override :type Boolean))
  (Utilities|SwitchHasAuthority
    (:Remote
      (Development|PrintString "@@@ UpdateOffset [BP_StatusWidgetComp_Default]: CALLED!" false (Variables|Debug|GetbDebugComponent))
      (if (CallFunction|IsOffsetAnimValid)
        (CallFunction|InitializeWidgetOffsetAnim Offset%20Y Animation%20Time "/NPCCore/Components/StatusWidget_AnimCurve_EaseInEaseOut_Inc.StatusWidget_AnimCurve_EaseInEaseOut_Inc")
        (Variables|WidgetOffset|SetbHoldWidgetOffset Hold%20Override)))))

(event Custom|ResetVerticalOffset
  (params (Animation%20Time :type Float%20%28double-precision%29))
  (bind _rendertransform_translation (Utilities|Struct|BreakWidgetTransform (Variables|/Script/UMG.Widget|GetRenderTransform_Translation_X (Variables|InternalData|GetNPCStatusWidget_UserObj))))
  (bind (_rendertransform_translation_x _rendertransform_translation_y) (Math|Vector2D|BreakVector2D _rendertransform_translation))
  (Utilities|SwitchHasAuthority
    (:Remote
      (Development|PrintString "@@@ ResetOffset [BP_StatusWidgetComp_Default]: CALLED!" false (Variables|Debug|GetbDebugComponent))
      (if (not (Variables|InternalData|GetbIsAnimating))
        (CallFunction|InitializeWidgetOffsetAnim _rendertransform_translation_y Animation%20Time "/NPCCore/Components/StatusWidget_AnimCurve_EaseInEaseOut_Dec.StatusWidget_AnimCurve_EaseInEaseOut_Dec")
        (Variables|WidgetOffset|SetbHoldWidgetOffset false)))))

(event EventTick
  (params (DeltaSeconds :type Float%20%28single-precision%29))
  (if (Variables|InternalData|GetbIsAnimating)
    (CallFunction|AnimateWidgetOffset DeltaSeconds (Variables|WidgetOffset|GetWidgetAnimTime) (Variables|WidgetOffset|GetWidgetAnimCurve))))

(fn AnimateWidgetOffset
  (params (Delta%20Time :type Float%20%28double-precision%29) (Animate%20Time :type Float%20%28double-precision%29) (Animate%20Curve :type Curve%20Float%20Object%20Reference))
  (bind _returnvalue (Class|CreativeBlueprintLibrary|AnimateFloatCurve Delta%20Time (Variables|InternalData|GetCurrentAnimateTime) Animate%20Curve Animate%20Time))
  (bind (_returnvalue_value _returnvalue_bcomplete) (Utilities|Struct|BreakFortCreativeAnimateFloatCurveResult _returnvalue))
  (if _returnvalue_bcomplete
    (Components|Tick|SetComponentTickEnabled false)
    (Variables|InternalData|SetbIsAnimating false)
    (Development|PrintString "@@@@@@ UpdateOffset [BP_StatusWidgetComp_Default]: Offset Anim COMPLETE!" false (Variables|Debug|GetbDebugComponent))
    (if (not (Variables|WidgetOffset|GetbHoldWidgetOffset))
      (Variables|/Game/Athena/AI/NPCs/Base/BP_PlayerPawn_NonParticipant_Boss.BP_PlayerPawn_NonParticipant_Boss_C|SetIsWidgetOffsetOverriden false (Variables|InternalData|GetNPCBossPawn)))
    (else
      (bind _returnvalue_2 (AI|GetNPCWidgetOffset (Fortnite|GetFirstLocalFortPlayerController true) (Variables|InternalData|GetNPCActor)))
      (Variables|/Script/UMG.Widget|SetRenderTransform_Translation_X (Utilities|Struct|MakeWidgetTransform (Math|Vector2D|MakeVector2D 0.0 (+ _returnvalue_2 (* _returnvalue_value (Variables|WidgetOffset|GetOffsetY))))) (Variables|InternalData|GetNPCStatusWidget_UserObj)))))

(fn InitializeWidgetOffsetAnim
  (params (Offset%20Y :type Float%20%28double-precision%29) (Animation%20Time :type Float%20%28double-precision%29) (Animation%20Curve :type Curve%20Float%20Object%20Reference))
  (Components|Tick|SetComponentTickEnabled true)
  (Variables|InternalData|SetCurrentAnimateTime 0.0)
  (Variables|InternalData|SetbIsAnimating true)
  (Variables|WidgetOffset|SetOffsetY Offset%20Y)
  (Variables|WidgetOffset|SetWidgetAnimTime Animation%20Time)
  (Variables|WidgetOffset|SetWidgetAnimCurve Animation%20Curve)
  (Variables|/Game/Athena/AI/NPCs/Base/BP_PlayerPawn_NonParticipant_Boss.BP_PlayerPawn_NonParticipant_Boss_C|SetIsWidgetOffsetOverriden true (Variables|InternalData|GetNPCBossPawn)))

(fn IsOffsetAnimValid
  (return (not (or (Variables|/Game/Athena/AI/NPCs/Base/BP_PlayerPawn_NonParticipant_Boss.BP_PlayerPawn_NonParticipant_Boss_C|GetIsWidgetOffsetOverriden (Variables|InternalData|GetNPCBossPawn)) (Variables|InternalData|GetbIsAnimating)))))
